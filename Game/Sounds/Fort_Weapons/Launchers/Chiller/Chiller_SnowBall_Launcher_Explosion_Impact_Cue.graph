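# Sound cue for the Chiller snowball launcher explosion impact, layering close, mid and far recordings by distance.
# A root mixer blends the three layers with the close one boosted, each having its own attenuation and sound class override.
# Each layer picks from four random variants, though quality-level switches drop mid and far to a single wave on some levels.
# The far layer adds a short random delay, and the cue's concurrency override prevents new instances at the limit.
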SoundCue /Game/Sounds/Fort_Weapons/Launchers/Chiller/Chiller_SnowBall_Launcher_Explosion_Impact_Cue
root=Mixer_5
nodes [
 {id=Attenuation_4, children [Modulator_13], properties {attenuationSettings=/Game/Sounds/Attenuation/Fort_Explosion_Medium_StW_Att}}
 {
  id=Attenuation_6
  children [Modulator_6]
  properties {
   attenuationSettings=/Game/Sounds/Attenuation/Fort_Explosion_Close_StW_Att
   attenuationOverrides {falloffDistance=900, attenuationShapeExtents {x=100}}
  }
 }
 {
  id=Attenuation_8
  children [Modulator_8]
  properties {
   attenuationSettings=/Game/Sounds/Attenuation/Fort_Explosion_Far_StW_Att
   attenuationOverrides {falloffDistance=3000, attenuationShapeExtents {x=1000}}
  }
 }
 {id=Delay_0, children [Attenuation_8], properties {delayMax=0.1}}
 {id=Mixer_2, children [SoundClass_3], properties {inputVolume [1]}}
 {id=Mixer_3, children [SoundClass_2], properties {inputVolume [1]}}
 {id=Mixer_4, children [SoundClass_0], properties {inputVolume [1]}}
 {id=Mixer_5, children [Mixer_2,Mixer_3,Mixer_4], properties {inputVolume [1.5,1,1]}}
 {id=Modulator_13, children [QualityLevel_1], properties {pitchMax=1, pitchMin=1, volumeMax=1, volumeMin=1}}
 {id=Modulator_6, children [Random_0], properties {pitchMax=1, pitchMin=1, volumeMax=1, volumeMin=1}}
 {id=Modulator_8, children [QualityLevel_0], properties {pitchMax=1, pitchMin=1, volumeMax=1, volumeMin=1}}
 {id=QualityLevel_0, children [Random_1, Random_1, WavePlayer_19, WavePlayer_19]}
 {id=QualityLevel_1, children [Random_3, Random_3, WavePlayer_13, WavePlayer_13]}
 {id=Random_0, children [WavePlayer_1, WavePlayer_6, WavePlayer_7, WavePlayer_10], properties {weights [1,1,1,1]}}
 {id=Random_1, children [WavePlayer_15, WavePlayer_16, WavePlayer_17, WavePlayer_18], properties {weights [1,1,1,1]}}
 {id=Random_3, children [WavePlayer_0, WavePlayer_11, WavePlayer_12, WavePlayer_14], properties {weights [1,1,1,1]}}
 {id=SoundClass_0, children [Delay_0], properties {soundClassOverride=&Fort_WeaponRanged_Fire_Distant}}
 {
  id=SoundClass_2
  children [Attenuation_4]
  properties {soundClassOverride=&Fort_WeaponRanged_Fire_Distant}
 }
 {
  id=SoundClass_3
  children [Attenuation_6]
  properties {soundClassOverride=&Fort_WeaponRanged_Fire_3P_PSM}
 }
 {id=WavePlayer_0, properties {soundWaveAssetPtr=&/EXPLO_Chiller_Snowball_Launcher_Mid_01}}
 {id=WavePlayer_1, properties {soundWaveAssetPtr=&/EXPLO_Chiller_Snowball_Launcher_Close_01}}
 {id=WavePlayer_10, properties {soundWaveAssetPtr=&/EXPLO_Chiller_Snowball_Launcher_Close_04}}
 {id=WavePlayer_11, properties {soundWaveAssetPtr=&/EXPLO_Chiller_Snowball_Launcher_Mid_02}}
 {id=WavePlayer_12, properties {soundWaveAssetPtr=&/EXPLO_Chiller_Snowball_Launcher_Mid_03}}
 {id=WavePlayer_13, properties {soundWaveAssetPtr=&EXPLO_Chiller_Snowball_Launcher_Mid_01}}
 {id=WavePlayer_14, properties {soundWaveAssetPtr=&/EXPLO_Chiller_Snowball_Launcher_Mid_04}}
 {id=WavePlayer_15, properties {soundWaveAssetPtr=&/EXPLO_Chiller_Snowball_Launcher_Far_01}}
 {id=WavePlayer_16, properties {soundWaveAssetPtr=&/EXPLO_Chiller_Snowball_Launcher_Far_02}}
 {id=WavePlayer_17, properties {soundWaveAssetPtr=&/EXPLO_Chiller_Snowball_Launcher_Far_03}}
 {id=WavePlayer_18, properties {soundWaveAssetPtr=&/EXPLO_Chiller_Snowball_Launcher_Far_04}}
 {id=WavePlayer_19, properties {soundWaveAssetPtr=&EXPLO_Chiller_Snowball_Launcher_Far_01}}
 {id=WavePlayer_6, properties {soundWaveAssetPtr=&/EXPLO_Chiller_Snowball_Launcher_Close_02}}
 {id=WavePlayer_7, properties {soundWaveAssetPtr=&/EXPLO_Chiller_Snowball_Launcher_Close_03}}
]
properties {
 bOverrideConcurrency=true
 maxDistance=11000
 soundClassObject=&Fort_WeaponMelee_PSM
 volumeMultiplier=1
 concurrencyOverrides {resolutionRule=PreventNew}
}
