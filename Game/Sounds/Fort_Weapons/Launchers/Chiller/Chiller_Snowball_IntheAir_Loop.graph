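# Sound cue for the in-flight sound of the Chiller launcher's snowball projectile.
# A mixer layers a looping flutter with two randomized one-shot starts, a jingle and a slightly quieter flutter, each picked from four variations.
# Only the loop runs through smoothed Doppler and its own overridden attenuation with distance low-pass filtering.
# The cue sets separate default attenuation and overrides concurrency to cap simultaneous instances.
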SoundCue /Game/Sounds/Fort_Weapons/Launchers/Chiller/Chiller_Snowball_IntheAir_Loop
root=Mixer_1
nodes [
 {
  id=Attenuation_0
  children [Doppler_1]
  properties {
   attenuationSettings=/Game/Sounds/Attenuation/Fort_Weapon_Projectile_InAir_Loop_Att
   bOverrideAttenuation=true
   attenuationOverrides {
    bAttenuateWithLPF=true
    falloffDistance=2500
    lPFFrequencyAtMin=500
    lPFRadiusMax=5000
    lPFRadiusMin=500
    reverbDistanceMax=10000
    reverbDistanceMin=250
    reverbWetLevelMax=0.6
    attenuationShapeExtents {x=150}
   }
  }
 }
 {id=Doppler_1, children [WavePlayer_0], properties {bUseSmoothing=true}}
 {id=Mixer_1, children [Attenuation_0, Random_0, Random_2], properties {inputVolume [1,1,0.75]}}
 {id=Random_0, children [WavePlayer_4,WavePlayer_5,WavePlayer_2,WavePlayer_1], properties {weights [1,1,1,1]}}
 {id=Random_2, children [WavePlayer_3, WavePlayer_6, WavePlayer_7, WavePlayer_10], properties {weights [1,1,1,1]}}
 {
  id=WavePlayer_0
  properties {bLooping=true, soundWaveAssetPtr=&/WPN_Chiller_Snowball_InTheAir_Flutter_Loop}
 }
 {id=WavePlayer_1, properties {soundWaveAssetPtr=&/WPN_Chiller_Snowball_InTheAir_Jingle_Start_04}}
 {id=WavePlayer_10, properties {soundWaveAssetPtr=&/WPN_Chiller_Snowball_InTheAir_Flutter_Start_04}}
 {id=WavePlayer_2, properties {soundWaveAssetPtr=&/WPN_Chiller_Snowball_InTheAir_Jingle_Start_03}}
 {id=WavePlayer_3, properties {soundWaveAssetPtr=&/WPN_Chiller_Snowball_InTheAir_Flutter_Start_01}}
 {id=WavePlayer_4, properties {soundWaveAssetPtr=&/WPN_Chiller_Snowball_InTheAir_Jingle_Start_01}}
 {id=WavePlayer_5, properties {soundWaveAssetPtr=&/WPN_Chiller_Snowball_InTheAir_Jingle_Start_02}}
 {id=WavePlayer_6, properties {soundWaveAssetPtr=&/WPN_Chiller_Snowball_InTheAir_Flutter_Start_02}}
 {id=WavePlayer_7, properties {soundWaveAssetPtr=&/WPN_Chiller_Snowball_InTheAir_Flutter_Start_03}}
]
properties {
 attenuationSettings=/Game/Sounds/Attenuation/Fort_Weapon_Launcher_Proj_InAir_Default_Att
 bOverrideConcurrency=true
 maxDistance=12750
 soundClassObject=&Fort_WeaponRanged_Fire_3P_PSM
 volumeMultiplier=1
 concurrencyOverrides {maxCount=3}
}
