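# Sound cue for the Chiller Snowball Launcher's firing sound, layered by distance.
# A top mixer blends a close layer under a third-person fire sound class with a distant-class submix of mid and far layers.
# Each layer picks randomly from three wave variants, passes through a modulator for variation, and has its own projectile attenuation.
# Local and global gunfire concurrency sets are assigned to the cue.
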SoundCue /Game/Sounds/Fort_Weapons/Launchers/Chiller/Chiller_Snowball_Launcher_Fire_Cue
root=Mixer_1
nodes [
 {
  id=Attenuation_10
  children [Modulator_14]
  properties {attenuationSettings=&Fort_Weapon_Projectile_Medium_Mono_Athena_Att}
 }
 {
  id=Attenuation_11
  children [Modulator_13]
  properties {attenuationSettings=&Fort_Weapon_Projectile_Distant_Mono_Athena_Att}
 }
 {id=Attenuation_9, children [Modulator_15], properties {attenuationSettings=&Fort_Weapon_Projectile_Close_Athena_Att}}
 {id=Mixer_1, children [SoundClass_7,SoundClass_6], properties {inputVolume [0.9,0.7]}}
 {id=Mixer_2, children [Attenuation_10, Attenuation_11], properties {inputVolume [0.6,0.4]}}
 {id=Modulator_13, children [Random_2], properties {pitchMax=1.1}}
 {id=Modulator_14, children [Random_0], properties {pitchMax=1.1, volumeMax=1, volumeMin=1}}
 {id=Modulator_15, children [Random_1]}
 {id=Random_0, children [WavePlayer_6,WavePlayer_7,WavePlayer_8], properties {weights [1,1,1]}}
 {id=Random_1, children [WavePlayer_0,WavePlayer_1,WavePlayer_2], properties {weights [1,1,1]}}
 {id=Random_2, children [WavePlayer_9,WavePlayer_4,WavePlayer_3], properties {weights [1,1,1]}}
 {id=SoundClass_6, children [Mixer_2], properties {soundClassOverride=&Fort_WeaponRanged_Fire_Distant}}
 {
  id=SoundClass_7
  children [Attenuation_9]
  properties {soundClassOverride=&Fort_WeaponRanged_Fire_3P_PSM}
 }
 {id=WavePlayer_0, properties {soundWaveAssetPtr=&/WPN_Chiller_Snowball_Launcher_Fire_01}}
 {id=WavePlayer_1, properties {soundWaveAssetPtr=&/WPN_Chiller_Snowball_Launcher_Fire_02}}
 {id=WavePlayer_2, properties {soundWaveAssetPtr=&/WPN_Chiller_Snowball_Launcher_Fire_03}}
 {id=WavePlayer_3, properties {soundWaveAssetPtr=&/WPN_Chiller_Snowball_Launcher_Fire_Far_03}}
 {id=WavePlayer_4, properties {soundWaveAssetPtr=&/WPN_Chiller_Snowball_Launcher_Fire_Far_02}}
 {id=WavePlayer_6, properties {soundWaveAssetPtr=&/WPN_Chiller_Snowball_Launcher_Fire_Mid_01}}
 {id=WavePlayer_7, properties {soundWaveAssetPtr=&/WPN_Chiller_Snowball_Launcher_Fire_Mid_02}}
 {id=WavePlayer_8, properties {soundWaveAssetPtr=&/WPN_Chiller_Snowball_Launcher_Fire_Mid_03}}
 {id=WavePlayer_9, properties {soundWaveAssetPtr=&/WPN_Chiller_Snowball_Launcher_Fire_Far_01}}
]
properties {
 maxDistance=27500
 volumeMultiplier=1
 attenuationOverrides {falloffDistance=5000, attenuationShapeExtents {x=1000}}
 concurrencyOverrides {bLimitToOwner=true, maxCount=3, resolutionRule=StopOldest}
 concurrencySet [&Weapon_Gunfire_Local_Conc, &Weapon_Gunfire_Global_Conc]
}
