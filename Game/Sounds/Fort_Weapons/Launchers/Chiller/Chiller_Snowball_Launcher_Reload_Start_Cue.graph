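# Sound cue for the start of reloading the Chiller snowball launcher.
# A single reload wave plays through a modulator and mixer at reduced volume, routed to the weapon foley sound class with weapon reload attenuation.
# Overridden concurrency allows one instance at a time, stopping the farthest before preventing new ones.
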SoundCue /Game/Sounds/Fort_Weapons/Launchers/Chiller/Chiller_Snowball_Launcher_Reload_Start_Cue
root=Mixer_0
nodes [
 {id=Mixer_0, children [Modulator_0], properties {inputVolume [1]}}
 {id=Modulator_0, children [WavePlayer_3]}
 {id=WavePlayer_3, properties {soundWaveAssetPtr=&/WPN_Chiller_Snowball_Launcher_Reload_01}}
]
properties {
 attenuationSettings=&Fort_Weapon_Reload_Att
 bOverrideConcurrency=true
 maxDistance=2250
 soundClassObject=&Fort_WeaponFoley
 volumeMultiplier=0.501
 attenuationOverrides {falloffDistance=5000, attenuationShapeExtents {x=1000}}
 concurrencyOverrides {maxCount=1, resolutionRule=StopFarthestThenPreventNew}
}
